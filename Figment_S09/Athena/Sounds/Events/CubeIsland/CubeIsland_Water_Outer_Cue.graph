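# CubeIsland_Water_Outer_Cue is a Sound Cue, apparently for outer water audio at the Cube Island event.
# Its graph has no nodes or root, so it produces no sound as it stands.
# It retains medium-priority event sound class routing, reduced volume, and non-spatialized attenuation override values with distance falloff.
# A concurrency override caps simultaneous instances at a few and stops the quietest when the limit is exceeded.
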SoundCue /Figment_S09/Athena/Sounds/Events/CubeIsland/CubeIsland_Water_Outer_Cue
root=null
nodes []
properties {
 bOverrideConcurrency=true
 maxDistance=5500
 soundClassObject=/Figment_S09/Athena/Sounds/Events/SoundClasses/Fort_Events_Priority_Medium
 volumeMultiplier=0.45
 attenuationOverrides {bSpatialize=false, falloffDistance=5000, nonSpatializedRadiusStart=500, stereoSpread=5000, attenuationShapeExtents {x=0}}
 concurrencyOverrides {maxCount=3, resolutionRule=StopQuietest}
}
